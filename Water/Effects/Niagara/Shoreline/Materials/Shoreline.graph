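# Unlit spline-mesh shoreline material that reads a Niagara render target along the strip, apparently experimental given its test parameter names and debug textures.
# UVs and vertex-color red form a segment coordinate for sampling NiagaraRT.
# Four neighboring texels feed GenerateASpline, whose smoothed result centers a SphereMask band across V.
# A vertical offset combines WaterBodyData terrain height, a bias and the scaled render-target value to conform vertices to terrain.
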Material /Water/Effects/Niagara/Shoreline/Materials/Shoreline
graph {
 properties {bUsedWithSplineMeshes=true, pixelDepthOffsetMode=PDOM_Legacy, shadingModel=MSM_Unlit, usedShadingModels=Unlit}
 nodes [
  n1 = TextureCoordinate()
  n2 = ScalarParameter() {defaultValue=255, parameterName=Quads}
  n3 = VertexColor()
  n4 = Constant()
  n5 = Divide(
   A=Add(A=Divide(A=@n1, B=@n2) {constB=255}, B=AppendVector(A=@n3.R, B=@n4))
   B=Divide(A=ScalarParameter() {defaultValue=60, parameterName=Segs}) {constB=255}
  )
  n6 = Multiply(A=ComponentMask(None=@n5) {r=true}, B=ScalarParameter() {defaultValue=16, parameterName=TTTT})
  n7 = ComponentMask(None=TextureCoordinate()) {g=true}
  Add(
   A=TextureSample(UVs=AppendVector(A=@n6, B=@n7)) {texture=/Engine/Engine_MI_Shaders/T_Base_Tile_Diffuse}.RGB
   B=AppendVector(A=@n5, B=Constant())
  )
  n8 = Add(A=Divide(A=ScalarParameter() {defaultValue=0.5, parameterName=offset}) {constA=0, constB=255}, B=@n5)
  n9 = TextureSampleParameter2D(UVs=@n8) {
   parameterName=NiagaraRT
   samplerType=SAMPLERTYPE_LinearColor
   texture=/Engine/Functions/Engine_MaterialFunctions02/PivotPainter2/Black_1x1_EXR_Texture
  }
  n10 = ScalarParameter() {parameterName=WPO test}
  n11 = MaterialFunctionCall!(Position) {materialFunction=/Water/Materials/Functions/WaterBodyData}
  n12 = Add(A=@n11.Terrain Z Height, B=ScalarParameter() {parameterName=Water Bias})
  n13 = ComponentMask(None=WorldPosition().XYZ) {b=true}
  Add(
   A=Multiply(A=Multiply(A=@n9.R, B=@n10), B=Constant3Vector() {constant=<rgba 0 0 1 0>})
   B=Multiply(A=Subtract(A=@n12, B=@n13), B=Constant3Vector() {constant=<rgba 0 0 1 0>})
  )
  ComponentMask(None=WorldPosition().XYZ) {g=true, r=true}
  Frac(None=Multiply(A=ComponentMask(None=@n5) {r=true}, B=ScalarParameter() {defaultValue=1, parameterName=fdfd}))
  Max(A=@n11.Water Z Height, B=@n11.Water Z Height)
  Multiply(A=@n3.R) {constB=255.01}
  Multiply(B=MaterialFunctionCall() {materialFunction=&TangentBasis}.X U)
  ScalarParameter() {parameterName=RRR}
  n14 = ComponentMask(None=TextureCoordinate()) {g=true}
  n15 = Constant2Vector() {g=1, r=256}
  n16 = Multiply(A=@n8, B=@n15)
  n17 = Floor(None=@n16)
  n18 = ComponentMask(None=Divide(A=Add(A=@n17) {constB=0.5}, B=@n15)) {r=true}
  n19 = Divide() {constA=-1, constB=255}
  n20 = TextureSampleParameter2D(UVs=AppendVector(A=Add(A=@n18, B=@n19), B=Constant() {r=0.5})) {
   parameterName=NiagaraRT
   samplerType=SAMPLERTYPE_LinearColor
   texture=&Black_1x1_EXR_Texture
  }
  n21 = TextureSampleParameter2D(UVs=AppendVector(A=@n18, B=Constant() {r=0.5})) {
   parameterName=NiagaraRT
   samplerType=SAMPLERTYPE_LinearColor
   texture=&Black_1x1_EXR_Texture
  }
  n22 = Divide() {constB=255}
  n23 = TextureSampleParameter2D(UVs=AppendVector(A=Add(A=@n18, B=@n22), B=Constant() {r=0.5})) {
   parameterName=NiagaraRT
   samplerType=SAMPLERTYPE_LinearColor
   texture=&Black_1x1_EXR_Texture
  }
  n24 = Divide() {constA=2, constB=255}
  n25 = TextureSampleParameter2D(UVs=AppendVector(A=Add(A=@n18, B=@n24), B=Constant() {r=0.5})) {
   parameterName=NiagaraRT
   samplerType=SAMPLERTYPE_LinearColor
   texture=&Black_1x1_EXR_Texture
  }
  n26 = MaterialFunctionCall!(t=ComponentMask(None=Frac(None=@n16)) {r=true}, P0=@n20.R, P1=@n21.R, P2=@n23.R, P3=@n25.R, Normal/Bi-Normal Cross product vector) {materialFunction=/Engine/Functions/Engine_MaterialFunctions02/WorldPositionOffset/GenerateASpline}
  SphereMask(
   A=Multiply(A=Subtract(A=@n14) {constB=0.5}) {constB=2}
   B=Divide(A=ComponentMask(None=@n26.T position) {r=true}, B=ScalarParameter() {defaultValue=768, parameterName=asdfasf})
   Radius=ScalarParameter() {defaultValue=0.1, parameterName=DRR}
  )
  TextureCoordinate()
  TextureSampleParameter2D(UVs=@n8) {
   parameterName=NiagaraRT
   samplerType=SAMPLERTYPE_LinearColor
   texture=&Black_1x1_EXR_Texture
  }
  TextureSample(UVs=@n5) {texture=/Engine/EngineDebugMaterials/HeatmapGradient}
  VertexColor()
 ]
}
