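# Niagara shoreline effect for water: a single GPU emitter that places long-lived sprite particles along a water body spline.
# An instantaneous burst spawns 256 particles, and a custom module samples the RT_Shoreline texture as spline position data, after which each initial position is stored.
# A 256x1 Grid2D collection is created as PhysicsGrid, but no listed module visibly uses it.
# Particle update only runs ParticleState, with no motion modules.
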NiagaraSystem /Water/Effects/Niagara/Shoreline/NiagaraShore_System
dependencies {
 usedDataInterfaces [/Script/Niagara.NiagaraDataInterfaceGrid2DCollection, /Script/Niagara.NiagaraDataInterfaceTexture]
 usedDynamicInputs []
 usedModules [&SystemState]
}
emitters {NiagaraShore {
  simTarget=GPUComputeSim
  emitterSpawnScript [/SetVariables(Emitter.PhysicsGrid=DataInterface() {propertyValues {
      ClearBeforeNonIterationStage=true
      NumAttributes=1
      NumCellsMaxAxis=3
      NumCellsX=256
      NumCellsY=1
      OverrideBufferFormat=Float
      PreviewAttribute=None
      SetGridFromMaxAxis=false
      bOverrideFormat=false
      bPreviewGrid=false
      RenderTargetUserParameter {parameter {name=None, typeDefHandle {registeredTypeIndex=4}, varData []}}
      WorldBBoxSize=<100 100>
     }}) {isSetParameters=true}]
  emitterUpdateScript [
   &EmitterState(Life Cycle Mode=::NewEnumerator1, Loop Duration=999, UseLoopDelay=false)
   &SpawnBurst_Instantaneous(Spawn Count=256, Use Spawn Probability=false)
  ]
  particleSpawnScript [
   /Niagara/Modules/Spawn/Initialization/InitializeParticle(
    Lifetime=&UniformRangedFloat(Minimum=999, Maximum=999)
    Mass=&UniformRangedFloat(Minimum=0.6, Maximum=1.2)
    Module.Write Color=true
    Module.Write Lifetime=true
    Module.Write Mass=true
    Module.Write Material Random=true
    Module.Write Position=false
    Module.Write Scale=false
    Module.Write SpriteRotation=false
    Module.Write SpriteSize=true
    Position=&SimulationPosition()
    Sprite Rotation=&NormalizedAngleToDegrees(NormalizedAngle=&UniformRangedFloat(Minimum=0, Maximum=1))
    Sprite Size=<256 256>
   )
   /Water/Effects/Niagara/Modules/NMS_Sample_Waterbody_Spline_RT(Spline Position RT=DataInterface() {propertyValues {
      Texture=&/RT_Shoreline
      TextureUserParameter {parameter {name=None, typeDefHandle {registeredTypeIndex=2}, varData []}}
     }})
   /SetVariables(Particles.InitialPosition:NiagaraPosition=$Particles.Position) {isSetParameters=true}
   /SetVariables() {isSetParameters=true}
  ]
  particleUpdateScript [&ParticleState()]
  rendererClasses [&NiagaraSpriteRendererProperties]
  renderers [{rendererClass=&NiagaraSpriteRendererProperties}]
 }}
summary {
 systemName=NiagaraShore_System
 emitters [{emitterName=NiagaraShore, simTarget=GPUComputeSim, rendererClasses [&NiagaraSpriteRendererProperties]}]
 userVariables []
}
